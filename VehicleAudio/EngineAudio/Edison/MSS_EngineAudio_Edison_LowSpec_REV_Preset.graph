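# Low-spec mono looping engine sound source for the Edison vehicle, wrapping the shared low-spec REV parent patch with Edison's REV asset and settings.
# Normalized speed, smoothed with affiliation-dependent timing and fast-rise/slow-fall interpolation, replaces RPM into the parent.
# Parent audio passes through a notch tracking the fundamental frequency and widening with the speed-derived RPM.
# That same value lowers a low-pass cutoff and attenuates volume as it rises.
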MetaSoundSource /VehicleAudio/EngineAudio/Edison/MSS_EngineAudio_Edison_LowSpec_REV_Preset
inputs [
 Blowoff.Sound {type=WaveAsset:Array, access=Constructor}
 Blowoff.Volume {type=Float, access=Constructor}
 Character.Affiliation {type=Enum:FortAudioAffiliation}
 Character.Health {type=Float}
 Character.UniqueID {type=Int32}
 Downshift {type=Trigger}
 EngineBlowoff {type=Trigger}
 Exhaust.Pops.Blowoff.Volume {type=Float, access=Constructor}
 Exhaust.Pops.Long {type=WaveAsset:Array, access=Constructor}
 Exhaust.Pops.Pitch {type=Float, access=Constructor}
 Exhaust.Pops.Shift.Volume {type=Float, access=Constructor}
 Exhaust.Pops.Short {type=WaveAsset:Array, access=Constructor}
 Exhaust.Pops.Volume {type=Float, access=Constructor}
 Exhaust.Throttle.EQ.FilterQ {type=Float, access=Constructor}
 Exhaust.Throttle.EQ.FilterType {type=Enum:BiquadFilterType, access=Constructor}
 Exhaust.Throttle.EQ.OffFreq {type=Float, access=Constructor}
 Exhaust.Throttle.EQ.OffGain {type=Float, access=Constructor}
 Exhaust.Throttle.EQ.OnFreq {type=Float, access=Constructor}
 Exhaust.Throttle.EQ.OnGain {type=Float, access=Constructor}
 Exhaust.Throttle.InterpTime {type=Time, access=Constructor}
 Exhaust.Throttle.Volume.Off {type=Float, access=Constructor}
 Exhaust.Throttle.Volume.On {type=Float, access=Constructor}
 Fort.Vehicle.Acceleration {type=Float}
 Fort.Vehicle.BlowoffStrength {type=Float}
 Fort.Vehicle.BonusSpeed {type=Float}
 Fort.Vehicle.EnableShifting {type=Bool}
 Fort.Vehicle.Gear {type=Int32}
 Fort.Vehicle.IsGrounded {type=Bool}
 Fort.Vehicle.Load {type=Float}
 Fort.Vehicle.NormalSpeed {type=Float}
 Fort.Vehicle.OnBlowoff {type=Trigger}
 Fort.Vehicle.OnDownshift {type=Trigger}
 Fort.Vehicle.OnThrottleEngaged {type=Trigger}
 Fort.Vehicle.OnThrottleReleased {type=Trigger}
 Fort.Vehicle.OnUpshift {type=Trigger}
 Fort.Vehicle.Pitch {type=Float}
 Fort.Vehicle.RPM {type=Float}
 Fort.Vehicle.Throttle {type=Float}
 GearChangeTime {type=Time, access=Constructor}
 IsLocal {type=Bool}
 Mono {type=Rev Asset, access=Constructor}
 REV.EnableShifting {type=Bool}
 REV.Gear {type=Int32}
 REV.RPM {type=Float}
 REV.Throttle {type=Float}
 REV.Velocity {type=Float}
 REVVolume {type=Float}
 Settings {type=Rev Settings, access=Constructor}
 Speed_Normalized {type=Float}
 UE.Source.OnPlay {type=Trigger}
 UE.Source.Orientation.Azimuth {type=Float}
 UE.Source.Orientation.Elevation {type=Float}
 Upshift {type=Trigger}
]
interfaces [Fort.Character:1.0, Fort.Vehicle:1.0, REV:1.0, UE.OutputFormat.Mono:1.0, UE.Source.Orientation:1.0, UE.Source:1.1]
outputs [
 Downshift {type=Trigger}
 Fundamental Frequency {type=Float}
 Gear {type=Int32}
 RPM {type=Float}
 SpeedNormalized {type=Float}
 Throttle {type=Float}
 "UE.OutputFormat.Mono.Audio:0" {type=Audio}
 Upshift {type=Trigger}
]
pages [Default {
  input_defaults {
   Blowoff.Volume=0
   Character.Affiliation=0
   Character.Health=0
   Character.UniqueID=0
   Downshift=null
   EngineBlowoff=null
   Exhaust.Pops.Blowoff.Volume=0
   Exhaust.Pops.Pitch=0
   Exhaust.Pops.Shift.Volume=0
   Exhaust.Pops.Volume=0
   Exhaust.Throttle.EQ.FilterQ=1
   Exhaust.Throttle.EQ.FilterType=4
   Exhaust.Throttle.EQ.OffFreq=20000
   Exhaust.Throttle.EQ.OffGain=0
   Exhaust.Throttle.EQ.OnFreq=20000
   Exhaust.Throttle.EQ.OnGain=0
   Exhaust.Throttle.InterpTime=0.1
   Exhaust.Throttle.Volume.Off=0
   Exhaust.Throttle.Volume.On=0
   Fort.Vehicle.Acceleration=0
   Fort.Vehicle.BlowoffStrength=0
   Fort.Vehicle.BonusSpeed=0
   Fort.Vehicle.EnableShifting=null
   Fort.Vehicle.Gear=0
   Fort.Vehicle.IsGrounded="<MetaSoundBuilderResult.SUCCEEDED: 0>"
   Fort.Vehicle.Load=0
   Fort.Vehicle.NormalSpeed=0
   Fort.Vehicle.OnBlowoff=null
   Fort.Vehicle.OnDownshift=null
   Fort.Vehicle.OnThrottleEngaged=null
   Fort.Vehicle.OnThrottleReleased=null
   Fort.Vehicle.OnUpshift=null
   Fort.Vehicle.Pitch=0
   Fort.Vehicle.RPM=0
   Fort.Vehicle.Throttle=0
   GearChangeTime=0.1
   IsLocal=null
   Mono=&/Engine_Edison_v5
   REV.EnableShifting=null
   REV.Gear=0
   REV.RPM=0
   REV.Throttle=0
   REV.Velocity=0
   REVVolume=6
   Settings=&/REVSettings_Edison
   Speed_Normalized=0
   UE.Source.OnPlay=null
   UE.Source.Orientation.Azimuth=0
   UE.Source.Orientation.Elevation=0
   Upshift=null
   Blowoff.Sound []
   Exhaust.Pops.Long []
   Exhaust.Pops.Short []
  }
  nodes [
   TriggerOnValueChange_0 = UE.Trigger On Value Change:Int32(Value=Convert.Int32:Enum:FortAudioAffiliation("Enum:FortAudioAffiliation"=^Character.Affiliation).Int32)
   TriggerAnyTriggerAny2_0 = "TriggerAny.Trigger Any (2)"(In 0=^UE.Source.OnPlay, In 1=@TriggerOnValueChange_0.Trigger)
   n1 = &TriggerCompareAffiliation(A=^Character.Affiliation, Compare=@TriggerAnyTriggerAny2_0.Out)
   TriggerRouteTriggerRouteFloat2_1 = "TriggerRoute.Trigger Route (Float, 2):Float"(Set 1=@n1.False, Set 0=@n1.True) {also {defaults {Value 0=1, Value 1=1}}}
   UE.InterpTo:Audio(Target=@TriggerRouteTriggerRouteFloat2_1.Value)
   TriggerOnValueChange_1 = UE.Trigger On Value Change:Int32(Value=Convert.Int32:Enum:FortAudioAffiliation("Enum:FortAudioAffiliation"=^Character.Affiliation).Int32)
   TriggerAnyTriggerAny2_1 = "TriggerAny.Trigger Any (2)"(In 0=^UE.Source.OnPlay, In 1=@TriggerOnValueChange_1.Trigger)
   n2 = &TriggerCompareAffiliation(A=^Character.Affiliation, Compare=@TriggerAnyTriggerAny2_1.Out)
   TriggerRouteTriggerRouteFloat2_0 = "TriggerRoute.Trigger Route (Float, 2):Float"(Set 1=@n2.False, Set 0=@n2.True) {also {defaults {Value 0=0.3, Value 1=0.6}}}
   n3 = UE.ConversionFloatToTime(In=@TriggerRouteTriggerRouteFloat2_0.Value)
   n4 = MapRange.MapRange:Float(In=Clamp.Clamp:Float(In=Max.Max:Float(B=^Fort.Vehicle.NormalSpeed, A=^Speed_Normalized).Value) {also {defaults {Max=0.9}}}.Value) {also {defaults {In Range B=0.9}}}
   n5 = UE.Reroute(Value=&InterpSpeed(FloatInput=UE.InterpTo:Audio(Interp Time=@n3.Out, Target=@n4.Out Value).Value) {also {defaults {FallingInterpSpeed=0.3, RisingInterpSpeed=10}}}.FloatOutput)
   MSS_EngineAudio_LowSpec_Parent_REV_0 = /VehicleAudio/Patches/MSS_EngineAudio_LowSpec_Parent_REV(
    Blowoff.Sound=^Blowoff.Sound
    Blowoff.Volume=^Blowoff.Volume
    Character.Affiliation=^Character.Affiliation
    Character.Health=^Character.Health
    Character.UniqueID=^Character.UniqueID
    Downshift=^Downshift
    EngineBlowoff=^EngineBlowoff
    Exhaust.Pops.Blowoff.Volume=^Exhaust.Pops.Blowoff.Volume
    Exhaust.Pops.Long=^Exhaust.Pops.Long
    Exhaust.Pops.Pitch=^Exhaust.Pops.Pitch
    Exhaust.Pops.Shift.Volume=^Exhaust.Pops.Shift.Volume
    Exhaust.Pops.Short=^Exhaust.Pops.Short
    Exhaust.Pops.Volume=^Exhaust.Pops.Volume
    Exhaust.Throttle.EQ.FilterQ=^Exhaust.Throttle.EQ.FilterQ
    Exhaust.Throttle.EQ.FilterType=^Exhaust.Throttle.EQ.FilterType
    Exhaust.Throttle.EQ.OffFreq=^Exhaust.Throttle.EQ.OffFreq
    Exhaust.Throttle.EQ.OffGain=^Exhaust.Throttle.EQ.OffGain
    Exhaust.Throttle.EQ.OnGain=^Exhaust.Throttle.EQ.OnGain
    Exhaust.Throttle.InterpTime=^Exhaust.Throttle.InterpTime
    Exhaust.Throttle.Volume.Off=^Exhaust.Throttle.Volume.Off
    Exhaust.Throttle.Volume.On=^Exhaust.Throttle.Volume.On
    Fort.Vehicle.Acceleration=^Fort.Vehicle.Acceleration
    Fort.Vehicle.BlowoffStrength=^Fort.Vehicle.BlowoffStrength
    Fort.Vehicle.BonusSpeed=^Fort.Vehicle.BonusSpeed
    Fort.Vehicle.EnableShifting=^Fort.Vehicle.EnableShifting
    Fort.Vehicle.Gear=^Fort.Vehicle.Gear
    Fort.Vehicle.IsGrounded=^Fort.Vehicle.IsGrounded
    Fort.Vehicle.Load=^Fort.Vehicle.Load
    Fort.Vehicle.NormalSpeed=^Fort.Vehicle.NormalSpeed
    Fort.Vehicle.OnBlowoff=^Fort.Vehicle.OnBlowoff
    Fort.Vehicle.OnDownshift=^Fort.Vehicle.OnDownshift
    Fort.Vehicle.OnThrottleEngaged=^Fort.Vehicle.OnThrottleEngaged
    Fort.Vehicle.OnThrottleReleased=^Fort.Vehicle.OnThrottleReleased
    Fort.Vehicle.OnUpshift=^Fort.Vehicle.OnUpshift
    Fort.Vehicle.Pitch=^Fort.Vehicle.Pitch
    Fort.Vehicle.Throttle=^Fort.Vehicle.Throttle
    GearChangeTime=^GearChangeTime
    IsLocal=^IsLocal
    Mono=^Mono
    REV.EnableShifting=^REV.EnableShifting
    REV.Gear=^REV.Gear
    REV.Throttle=^REV.Throttle
    REV.Velocity=^REV.Velocity
    REVVolume=^REVVolume
    Settings=^Settings
    Speed_Normalized=^Speed_Normalized
    UE.Source.OnPlay=^UE.Source.OnPlay
    UE.Source.Orientation.Azimuth=^UE.Source.Orientation.Azimuth
    UE.Source.Orientation.Elevation=^UE.Source.Orientation.Elevation
    Upshift=^Upshift
    Fort.Vehicle.RPM=@n5.Value
    REV.RPM=@n5.Value
   )
   n6 = VariableAccessor.Float(Variable=VariableMutator.Float(Variable=InitVariable.Float().Variable, Value=@n5.Value).Variable)
   n7 = VariableAccessor.Float(Variable=@n6.Variable)
   n8 = VariableAccessor.Float(Variable=@n7.Variable)
   BiquadFilter_1 = UE.Biquad Filter:Audio(
    In=@MSS_EngineAudio_LowSpec_Parent_REV_0."UE.OutputFormat.Mono.Audio:0"
    Bandwidth=MapRange.MapRange:Float(In=/VehicleAudio/Patches/EaseOutQuad(In=@n8.Value).Out) {also {defaults {Out Range A=0.1, Out Range B=1.5}}}.Out Value
    Cutoff Frequency=UE.Multiply:Float(PrimaryOperand=@MSS_EngineAudio_LowSpec_Parent_REV_0.Fundamental Frequency) {also {defaults {AdditionalOperands=26}}}.Out
   ) {also {defaults {Type=3}}}
   BiquadFilter_0 = UE.Biquad Filter:Audio(
    In=@BiquadFilter_1.Out
    Cutoff Frequency=MapRange.MapRange:Float(In=@n7.Value) {also {defaults {Out Range A=20000, Out Range B=3000}}}.Out Value
   ) {also {defaults {Bandwidth=2, Type=0}}}
   Multiply_0 = UE.Multiply:Audio by Float(
    PrimaryOperand=@BiquadFilter_0.Out
    AdditionalOperands=UE.Multiply:Float(AdditionalOperands=/FortMetasound/Patches/MapRange_DecibelOutput(In=@n6.Value) {also {defaults {Out Range B=-8}}}.Out Value).Out
   )
  ]
  links [
   {dst="@Downshift", src=MSS_EngineAudio_LowSpec_Parent_REV_0.Downshift}
   {dst="@Fundamental Frequency", src=MSS_EngineAudio_LowSpec_Parent_REV_0.Fundamental Frequency}
   {dst="@Gear", src=MSS_EngineAudio_LowSpec_Parent_REV_0.Gear}
   {dst="@RPM", src=MSS_EngineAudio_LowSpec_Parent_REV_0.RPM}
   {dst="@SpeedNormalized", src=MSS_EngineAudio_LowSpec_Parent_REV_0.SpeedNormalized}
   {dst="@Throttle", src=MSS_EngineAudio_LowSpec_Parent_REV_0.Throttle}
   {dst="@Upshift", src=MSS_EngineAudio_LowSpec_Parent_REV_0.Upshift}
   {dst="@UE.OutputFormat.Mono.Audio:0", src=Multiply_0.Out}
  ]
 }]
properties {
 assetImportData=&self:AssetImportData
 attenuationSettings=&Vehicle_Engine_Att
 audioPropertiesSheet=&APS_Cosmetic_Vehicle_Engine
 bLooping=true
 duration=10000
 maxDistance=12000
 priority=80
 qualitySetting=Vehicles_REV_Mono
 sampleRate=48000
 soundClassObject=&Fort_Vehicles_Engine
 soundSubmixObject=&Fortnite_SFX
 volume=5.012
 concurrencySet [&Conc_Veh_EngineAudio_LowSpec, &Conc_Veh_EngineAudio_LowSpec_Actor]
 modulationSettings {
  highpassRouting=Union
  lowpassRouting=Union
  pitchRouting=Union
  volumeRouting=Union
  highpassModulationDestination {bEnableModulation=true, modulators [&PP_FortVehicles_Engine_HPF]}
  lowpassModulationDestination {bEnableModulation=true, modulators [&Patch_FortVehicles_Engine_LPF]}
  pitchModulationDestination {bEnableModulation=true, modulators [&DM_Bus_Vehicle_Engine_Pitch]}
  volumeModulationDestination {
   bEnableModulation=true
   modulators [&DM_VehicleEnginePatch, &Patch_FortVehicles_Engine_Vol]
  }
 }
}
